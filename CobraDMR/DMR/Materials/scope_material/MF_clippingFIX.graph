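# Material function for the DMR scope material that outputs a World Position Offset to fix first-person mesh clipping.
# It rescales camera-relative, ScreenOffset-shifted positions perpendicular to view using an FOV-to-Param_1 tangent ratio.
# Points beyond the near plane then compress toward it to a tenth of their distance, via custom HLSL.
# A static switch and scalar input, defaulting to zero, gate the result; scope collection parameters and CameraOffset are unconnected.
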MaterialFunction /CobraDMR/DMR/Materials/scope_material/MF_clippingFIX
graph {nodes [
  CollectionParameter() {collection=&/MPC_DMR_Scope, parameterName=scope_cameraFov}
  CollectionParameter() {collection=&MPC_DMR_Scope, parameterName=Scope_WP_switch}
  n1 = MaterialFunctionCall() {materialFunction=&CameraDirectionVector}
  n2 = Add(
   A=WorldPosition() {worldPositionShaderOffset=WPT_CameraRelative}.XYZ
   B=Transform(None=VectorParameter() {parameterName=ScreenOffset, defaultValue=<rgba 0 0 0 1>}.RGB) {transformSourceType=TRANSFORMSOURCE_View}
  )
  n3 = Subtract(A=Multiply(A=@n1.Result, B=DotProduct(A=@n1.Result, B=@n2)), B=@n2)
  n4 = ScalarParameter() {defaultValue=80, parameterName=Param_1}
  n5 = OneMinus(None=Divide(
    A=MaterialFunctionCall() {materialFunction=/Engine/Functions/Engine_MaterialFunctions02/Utility/FOV}."FOV "
    B=Tangent!(None=@n4) {period=720}
   ))
  n6 = Add(A=Add(A=Multiply(A=@n3, B=@n5), B=@n2), B=WorldPosition() {worldPositionShaderOffset=WPT_CameraRelative}.XYZ)
  n7 = Normalize(VectorInput=@n6)
  n8 = MaterialFunctionCall() {materialFunction=&CameraDirectionVector}
  n9 = Multiply(A=Divide(A=@n7, B=DotProduct(A=@n7, B=@n8.Result)), B=Custom!(None) {code=View.NearPlane, description=NearClip, outputType=CMOT_Float1})
  n10 = Subtract(A=@n6, B=@n9)
  MaterialExpressionDotProduct_2 = DotProduct(A=@n10, B=MaterialFunctionCall() {materialFunction=&CameraDirectionVector}.Result)
  n11 = Custom!(X=@MaterialExpressionDotProduct_2) {
   code="return step(X, 0.0);"
   description=Plus
   outputType=CMOT_Float1
   inputs [{inputName=X, input {expression=&self:MaterialExpressionDotProduct_2}}]
  }
  n12 = Max(A=@n11, B=Constant() {r=0.1})
  n13 = Subtract(A=Add(A=Multiply(A=@n12, B=@n10), B=@n9), B=WorldPosition() {worldPositionShaderOffset=WPT_CameraRelative}.XYZ)
  n14 = Constant()
  n15 = FunctionInput(Preview=@n14) {bUsePreviewValueAsDefault=true, inputType=FunctionInput_Scalar}
  FunctionOutput(None=StaticSwitch(
    True=LinearInterpolate(B=@n13, Alpha=@n15) {constB=0}
    False=@n14
    Value=StaticBoolParameter() {defaultValue=true, parameterName=use WPO fix clipping}
   ))
  MaterialFunctionCall(Offset Amount=ScalarParameter() {defaultValue=5}, World Position=WorldPosition().XYZ) {materialFunction=/Engine/Functions/Engine_MaterialFunctions02/WorldPositionOffset/CameraOffset}
 ]}
