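# Translucent, lit lens material for the DMR weapon scope, drawing reticle, crosshair, edge ring and tinted glass.
# BumpOffset parallax fakes tube depth in the view masks and reticle UVs, with a static switch to viewport-scaled screen-space UVs.
# Cubemap reflections, Fresnel, scratches, dirt and sun-direction terms build the glass surface.
# View masks and reticle shapes drive opacity, and a quality switch limits noise-modulated refraction to Epic.
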Material /CobraDMR/DMR/Materials/scope_material/M_DMR_ScopeGlass
graph {
 properties {
  bEnableResponsiveAA=true
  bUsedWithMeshDeformer=true
  bUsedWithSkeletalMesh=true
  blendMode=BLEND_Translucent
  pixelDepthOffsetMode=PDOM_Legacy
  refractionMethod=RM_IndexOfRefraction
  translucencyLightingMode=TLM_VolumetricDirectional
  translucencyPass=MTP_AfterMotionBlur
  usedShadingModels=Default Lit
 }
 nodes [
  Constant() {r=0.65}
  LinearInterpolate(
   B=NamedRerouteUsage()
   Alpha=SphereMask(A=TextureCoordinate(), B=Constant() {r=0.5}, Radius=Constant() {r=0.452}, Hardness=Constant() {r=0.99})
  )
  n1 = NamedRerouteUsage()
  n2 = Transform(None=NamedRerouteUsage())
  n3 = Fresnel(ExponentIn=Constant() {r=3}, Normal=@n2)
  n4 = Multiply(A=NamedRerouteUsage(), B=LinearInterpolate(Alpha=Power(Base=@n3) {constExponent=1.2}) {constA=0.004, constB=0.25}) {constB=0.01}
  n5 = Multiply(A=TextureCoordinate()) {constB=2}
  n6 = Multiply(
   A=LinearInterpolate(A=Constant(), B=Constant() {r=0.03}, Alpha=TextureSample(UVs=@n5) {texture=&/T_Marksman_scratches}.R)
   B=Power(Base=NamedRerouteUsage())
  )
  LinearInterpolate(
   A=Add(A=@n1, B=Add(A=@n6, B=@n4))
   B=EyeAdaptationInverse(LightValueInput=VectorParameter() {group=Scope, parameterName=CrossHair_Color, defaultValue=<rgba 0 2 0.4561 1>}.RGB).EyeAdaptationInverse
   Alpha=NamedRerouteUsage()
  )
  n7 = Multiply(
   A=TextureSample(UVs=TextureCoordinate() {uTiling=3, vTiling=3}) {texture=/Game/Athena/Environments/Island/Textures/Dirt_D}.RGB
   B=Constant() {r=0.05}
  )
  n8 = BumpOffset(Coordinate=TextureCoordinate(), Height=NamedRerouteUsage())
  n9 = ScalarParameter() {defaultValue=0.69, group=Scope, parameterName=view_size}
  n10 = SphereMask(A=@n8, B=Constant() {r=0.5}, Radius=@n9, Hardness=Constant() {r=0.99})
  n11 = SphereMask(A=@n8, B=Constant() {r=0.5}, Radius=Subtract(A=@n9) {constB=0.002}, Hardness=Constant() {r=0.92})
  n12 = Multiply(A=Saturate(None=Subtract(A=@n10, B=@n11)), B=NamedRerouteUsage())
  n13 = VectorParameter() {group=Scope, parameterName=ScopeEdgeColor, defaultValue=<rgba 0 0 0.013 1>}
  n14 = VectorParameter() {group=Scope, parameterName=Reticle_Color, defaultValue=<rgba 0 0 0 1>}
  n15 = Add(A=Multiply(A=NamedRerouteUsage(), B=@n14.RGB), B=Multiply(A=NamedRerouteUsage(), B=@n14.RGB))
  n16 = TextureCoordinate()
  n17 = Constant() {r=0.49}
  n18 = Constant() {r=0.5}
  n19 = Constant() {r=0.2}
  n20 = Add(
   A=Add(A=Multiply(A=@n12, B=@n13.RGB), B=Saturate(None=@n15))
   B=LinearInterpolate(
    A=Constant3Vector() {constant=<rgba 0.003 0.003 0.003 0>}
    B=VectorParameter() {group=Scope, parameterName=ScopeGlassColor, defaultValue=<rgba 0.0284 0.0452 0.0152 1>}.RGB
    Alpha=SphereMask(A=@n16, B=@n17, Radius=@n18, Hardness=@n19)
   )
  )
  n21 = SphereMask(A=TextureCoordinate(), B=Constant() {r=0.49}, Radius=Constant() {r=0.5}, Hardness=Constant() {r=0.9})
  LinearInterpolate(
   A=Add(A=@n1, B=LinearInterpolate(A=@n7, B=@n20, Alpha=@n21))
   B=VectorParameter() {group=Scope, parameterName=CrossHair_Color, defaultValue=<rgba 0 2 0.4561 1>}.RGB
   Alpha=NamedRerouteUsage()
  )
  MaterialFunctionCall!(In=ScalarParameter() {defaultValue=1, parameterName=ScopeGlassWP}) {materialFunction=&/MF_clippingFIX}
  n22 = NamedRerouteUsage()
  Multiply(
   A=Power(Base=NamedRerouteUsage()) {constExponent=600}
   B=Multiply(A=Power(Base=@n22), B=Constant3Vector() {constant=<rgba 0.5 0.45 0.4511 0>})
  ) {constB=0.3}
  n23 = TextureCoordinate()
  n24 = NamedRerouteDeclaration(None=ScalarParameter() {defaultValue=-80, group=Scope, parameterName=View_Depth}) {name=View_Depth, nodeColor=<rgba 0 0.0625 1 1>}
  NamedRerouteDeclaration(None=SphereMask(
    A=BumpOffset(Coordinate=@n23, Height=@n24)
    B=Constant() {r=0.5}
    Radius=ScalarParameter() {defaultValue=0.69, group=Scope, parameterName=view_size}
    Hardness=ScalarParameter() {defaultValue=0.2, group=Scope, parameterName=View_Hardness}
   )) {name=scope back view, nodeColor=<rgba 1 0 0.5391 1>}
  NamedRerouteDeclaration(None=SphereMask(A=TextureCoordinate(), B=Constant() {r=0.5}, Radius=Constant() {r=0.46}, Hardness=Constant() {r=0.2})) {name=scope front view, nodeColor=<rgba 0.6172 1 0 1>}
  n25 = TextureSampleParameter2D(UVs=StaticSwitchParameter(True=NamedRerouteUsage(), False=NamedRerouteUsage()) {group=Scope, parameterName=UseScreenSpaceUVs, sortPriority=0}) {
   desc="R- Reticle  B-Crosshair"
   group=Scope
   parameterName=T_Reticle
   samplerSource=SSM_Clamp_WorldGroupSettings
   texture=&T_BlacksRGB
  }
  NamedRerouteDeclaration(None=@n25.R) {name=Reticle_Red, nodeColor=<rgba 0.8516 1 0 1>}
  NamedRerouteDeclaration(None=@n25.B) {name=Reticle_Blue, nodeColor=<rgba 1 0.9492 0 1>}
  n26 = ViewSize()
  n27 = Subtract(A=ScreenPosition().PixelPosition, B=Ceil(None=Multiply(B=@n26) {constA=0.5}))
  n28 = ComponentMask(None=ViewSize()) {g=true}
  n29 = Multiply(A=@n28) {constB=0.025}
  n30 = ViewProperty()
  n31 = Subtract(A=Divide(A=ComponentMask(None=@n30.Property) {r=true}, B=ComponentMask(None=@n30.Property) {g=true}))
  n32 = Add(A=@n29, B=Multiply(A=@n31, B=Multiply(A=@n29) {constB=0.2215}))
  NamedRerouteDeclaration(None=Add(A=Divide(A=@n27, B=@n32)) {constB=0.5}) {name=CrosshairUV_Screenspace, nodeColor=<rgba 1 0 0.0937 1>}
  n33 = MaterialFunctionCall(UVs=TextureCoordinate(), Texture Scale=ScalarParameter() {defaultValue=0.06, group=Scope, parameterName=Crosshair_Scale}) {materialFunction=&ScaleUVsByCenter}
  n34 = ScalarParameter() {defaultValue=0.5, group=Scope, parameterName=Crosshair_Rotation}
  NamedRerouteDeclaration(None=BumpOffset(
    Coordinate=MaterialFunctionCall(UVs=@n33.UVs, "Rotation Angle (0-1)"=@n34) {materialFunction=&CustomRotator}.Rotated Values
    Height=ScalarParameter() {defaultValue=-17, group=Scope, parameterName=Crosshair_depth}
   )) {name=CrosshairUV_TexCoord, nodeColor=<rgba 1 0.75 0 1>}
  n35 = Multiply(A=@n28) {constB=0.409}
  n36 = Add(A=@n35, B=Multiply(A=Multiply(A=@n35) {constB=0.2215}, B=@n31))
  NamedRerouteDeclaration(None=Add(A=Divide(A=@n27, B=@n36)) {constB=0.5}) {name=ReticleUV_Screenspace, nodeColor=<rgba 0.0781 0 1 1>}
  n37 = MaterialFunctionCall(UVs=TextureCoordinate(), Texture Scale=ScalarParameter() {defaultValue=0.95, group=Scope, parameterName=Reticle_Scale}) {materialFunction=&ScaleUVsByCenter}
  n38 = ScalarParameter() {defaultValue=0.5, group=Scope, parameterName=Reticle_Rotation}
  NamedRerouteDeclaration(None=BumpOffset(
    Coordinate=MaterialFunctionCall(UVs=@n37.UVs, "Rotation Angle (0-1)"=@n38) {materialFunction=&CustomRotator}.Rotated Values
    Height=ScalarParameter() {defaultValue=-15, group=Scope, parameterName=Reticle_depth}
   )) {name=ReticleUV_TexCoord, nodeColor=<rgba 0.8516 1 0 1>}
  n39 = TextureCoordinate()
  n40 = Constant() {r=1.1}
  NamedRerouteDeclaration(None=TextureSample(UVs=MaterialFunctionCall(UVs=@n39, Texture Scale=@n40) {materialFunction=&ScaleUVsByCenter}.UVs) {samplerType=SAMPLERTYPE_Normal, texture=&/T_MarksmanGlass_N}.RGB) {name=Scope_normal, nodeColor=<rgba 0 0.9062 1 1>}
  n41 = NamedRerouteUsage()
  n42 = Normalize(VectorInput=Transform(None=@n41))
  n43 = TextureSampleParameterCube!(UVs=LinearInterpolate(A=ReflectionVectorWS(CustomWorldNormal=@n42), B=ReflectionVectorWS()) {constAlpha=0.2}, Apply View MipBias) {group=Scope, parameterName=Glass_CubeMap, texture=/Game/Packages/Fortress_Sky/SkyDome/MaterialInstances/SkyDome_Apollo/Cubemap_Apollo_Day}
  n44 = Multiply(A=@n43.RGB, B=Fresnel(ExponentIn=Constant() {r=0.125}, Normal=@n41))
  n45 = Constant() {r=1}
  NamedRerouteDeclaration(None=Multiply(A=Power(Base=@n44, Exp=@n45), B=ScalarParameter() {defaultValue=0.5, group=Scope, parameterName=intensity of reflection})) {name=Cubemap, nodeColor=<rgba 0 1 0.2031 1>}
  n46 = TextureSampleParameter2D(UVs=StaticSwitchParameter(True=NamedRerouteUsage(), False=NamedRerouteUsage()) {group=Scope, parameterName=UseScreenSpaceUVs, sortPriority=0}) {group=Scope, parameterName=T_Cross_M, samplerSource=SSM_Clamp_WorldGroupSettings, texture=&T_BlacksRGB}
  n47 = StaticSwitch(True=@n46.R, False=Constant(), Value=StaticBoolParameter() {group=Scope, parameterName="Use  Crosshair"})
  n48 = Add(A=@n25.R, B=@n25.R)
  n49 = Saturate(None=Max(A=@n47, B=Saturate(None=@n48)))
  NamedRerouteDeclaration(None=OneMinus(None=@n49)) {name=crossponit, nodeColor=<rgba 0.5703 0 1 1>}
  n50 = NamedRerouteUsage()
  n51 = NamedRerouteUsage()
  n52 = Saturate(None=Multiply(A=Multiply(A=@n50, B=@n51)) {constB=8})
  n53 = Add(A=Power(Base=OneMinus(None=@n52)), B=ScalarParameter() {defaultValue=0.1, group=Scope, parameterName="Scope Inter opacity  indensity"}) {constB=0.1}
  NamedRerouteDeclaration(None=Saturate(None=Add(A=@n49, B=@n53))) {name=Opacity, nodeColor=<rgba 0.4531 0 1 1>}
  n54 = AtmosphericLightVector()
  NamedRerouteDeclaration(None=Saturate(None=DotProduct(A=@n2, B=@n54))) {name=Specular, nodeColor=<rgba 1 0 0.4453 1>}
  NamedRerouteDeclaration(None=Constant()) {name=HightLight, nodeColor=<rgba 0 1 0.8125 1>}
  NamedRerouteDeclaration(None=@n47) {name=Crosshair, nodeColor=<rgba 0.0781 0 1 1>}
  NamedRerouteUsage()
  n55 = Constant() {r=1}
  n56 = MaterialFunctionCall() {materialFunction=&VectorToRadialValue}
  n57 = Constant() {r=2}
  n58 = Multiply(
   A=TextureSample(UVs=Multiply(A=@n56.Radial Coordinates, B=@n57)) {texture=&AtmosphericCloudNoise01}.RGB
   B=Constant() {r=1}
  )
  n59 = Saturate(None=Multiply(A=NamedRerouteUsage(), B=Constant() {r=4.2}))
  QualitySwitch(
   Default=@n55
   Low=@n55
   High=@n55
   Medium=@n55
   Epic=LinearInterpolate(
    A=LinearInterpolate(A=@n58, Alpha=@n59) {constA=1.02}
    Alpha=ScalarParameter() {defaultValue=0.98, group=Scope, parameterName=Refraction_Indensity}
   ) {constAlpha=0.98}
  )
  VectorParameter() {parameterName=Rarity_Color, defaultValue=<rgba 0 0 0 1>}
 ]
}
