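# Opaque Default Lit material for the DMR scope glass, usable on skeletal meshes.
# A high-power view-to-normal dot product blends two dark olive tint colors, while a tiled scratch texture blends between a low and a high scalar, likely for roughness.
# A scalar switch blends toward an object-position-minus-world-position offset, which would collapse the glass to its pivot as world position offset.
# The scope collection parameter node feeds none of these blends.
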Material /CobraDMR/DMR/Materials/scope_material/M_DMR_ScopeGlass_opaque
graph {
 properties {
  bUsedWithMeshDeformer=true
  bUsedWithSkeletalMesh=true
  pixelDepthOffsetMode=PDOM_Legacy
  translucencyLightingMode=TLM_VolumetricDirectional
  usedShadingModels=Default Lit
 }
 nodes [
  CollectionParameter() {collection=&/MPC_DMR_Scope, parameterName=Scope_WP_switch}
  n1 = DotProduct(A=CameraVectorWS(), B=VertexNormalWS())
  LinearInterpolate(
   A=VectorParameter() {parameterName=Glassbaseacolor, defaultValue=<rgba 0.0162 0.02 0 1>}.RGB
   B=VectorParameter() {parameterName=GlassInterColor, defaultValue=<rgba 0.0989 0.1 0 1>}.RGB
   Alpha=Saturate(None=Power(Base=@n1) {constExponent=60})
  )
  LinearInterpolate(
   A=Constant() {r=0.05}
   B=Constant() {r=1.5}
   Alpha=TextureSample(UVs=TextureCoordinate() {uTiling=4, vTiling=4}) {texture=&/T_Marksman_scratches}.R
  )
  LinearInterpolate(B=Subtract(A=ObjectPositionWS(), B=WorldPosition().XYZ), Alpha=ScalarParameter() {parameterName=WP_switch_scopGlass})
 ]
}
