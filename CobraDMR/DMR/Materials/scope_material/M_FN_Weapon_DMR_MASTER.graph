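# Masked master material for the DMR weapon, layering the shared weapon master, wrap customization and ADS scope functions.
# Weapon attributes gain clipping-fix WPO and an optional opacity mask, both driven by WP_switch, before customization uses per-weapon bounds and safe zones.
# Elemental and overheat effects follow, skipped on mobile, then the scope function and optional vertex-color hiding through WPO.
# Zero-alpha lerps tie wrap parameters into Base Color without changing it.
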Material /CobraDMR/DMR/Materials/scope_material/M_FN_Weapon_DMR_MASTER
graph {
 properties {
  bAutomaticallySetUsageInEditor=false
  bUseMaterialAttributes=true
  bUsedWithClothing=true
  bUsedWithMeshDeformer=true
  bUsedWithMorphTargets=true
  bUsedWithNanite=true
  bUsedWithSkeletalMesh=true
  blendMode=BLEND_Masked
  pixelDepthOffsetMode=PDOM_Legacy
  refractionMethod=RM_IndexOfRefraction
  usedShadingModels=Default Lit
 }
 nodes [
  Constant()
  n1 = MaterialFunctionCall!(In) {materialFunction=/Game/Packages/Fortress_SharedMaterials/Base_Material_Functions/BaseCustomizationFunctions/MF_FN_Weapon_MASTER}
  Reroute(None=@n1.Material Result)
  n2 = GetMaterialAttributes(None=@n1.Material Result) {attributeGetTypes [#1]}
  n3 = VectorParameter() {group=Weapon Specific Wrap Adjustments, parameterName=DO NOT OVERRIDE-OBMAX, sortPriority=50, defaultValue=<rgba 0 0 0 1>}
  n4 = VectorParameter() {group=Weapon Specific Wrap Adjustments, parameterName=DO NOT OVERRIDE-OBMIN, sortPriority=51, defaultValue=<rgba 0 0 0 1>}
  n5 = Add(A=@n3.RGB, B=@n4.RGB)
  n6 = VectorParameter() {group=Weapon Specific Wrap Adjustments, parameterName=DO NOT OVERRIDE-OBCENTER, sortPriority=55, defaultValue=<rgba 0 0 0 1>}
  n7 = VectorParameter() {group=Weapon Specific Wrap Adjustments, parameterName=DO NOT OVERRIDE-SZ-Back, sortPriority=100, defaultValue=<rgba 0.3 0 0 0.2>}
  n8 = VectorParameter() {group=Weapon Specific Wrap Adjustments, parameterName=DO NOT OVERRIDE-SZ-Middle, sortPriority=101, defaultValue=<rgba 0 0 0 0.2>}
  n9 = VectorParameter() {group=Weapon Specific Wrap Adjustments, parameterName=DO NOT OVERRIDE-SZ-Front, sortPriority=102, defaultValue=<rgba -0.3 0 0 0.2>}
  n10 = Add(A=Add(A=@n7.RGB, B=@n8.RGB), B=@n9.RGB)
  n11 = VectorParameter() {group=Weapon Specific Wrap Adjustments, parameterName=DO NOT OVERRIDE-Virtual-Pivot, sortPriority=110, defaultValue=<rgba 0 0 0 0.5>}
  n12 = LinearInterpolate(A=@n2.Base Color, B=Add(A=Add(A=@n5, B=@n6.RGB), B=Add(A=@n10, B=@n11.RGB))) {constAlpha=0}
  n13 = ScalarParameter() {group=Weapon Specific Wrap Adjustments, parameterName=IsPickaxe, sortPriority=200}
  n14 = ScalarParameter() {parameterName=WP_switch}
  n15 = TextureSampleParameter2D() {group=Opacity Mask, parameterName=MeshOpacityMask, samplerType=SAMPLERTYPE_Masks, texture=&T_BlackMask}
  n16 = Multiply(
   A=StaticComponentMaskParameter(None=@n15.RGBA) {defaultR=true, group=Opacity Mask, parameterName=MeshOpacityMaskChannel}
   B=ScalarParameter() {defaultValue=1, group=Opacity Mask, parameterName=MeshOpacityMask_strength}
  ) {constB=10}
  n17 = LinearInterpolate(B=OneMinus(None=Saturate(None=@n16)), Alpha=@n14) {constA=1}
  n18 = Constant() {r=1}
  n19 = VectorParameter() {group=Weapon Specific Wrap Adjustments, parameterName=WeaponForwardAxis, sortPriority=200, defaultValue=<rgba 0 1 0 1>}
  n20 = Constant()
  n21 = MaterialFunctionCall!(
   Material=SetMaterialAttributes!(
    MaterialAttributes=@n2.MaterialAttributes
    Base Color=LinearInterpolate(A=@n12, B=@n13) {constAlpha=0}
    World Position Offset=MaterialFunctionCall!(In=@n14) {materialFunction=&/MF_clippingFIX}.Result
    Opacity Mask=StaticSwitchParameter(True=@n17, False=@n18) {defaultValue=true, group=Opacity Mask, parameterName=UseMaskHideMesh}
   ) {attributeSetTypes [#1, #2, #3]}
   ------------------
   Object Bounds Center Adjustment=@n6.RGB
   Object Bounds Min Adjustment=@n4.RGB
   Wrap ScaleMultiplier
   Rotation X Axis
   "Customization Mask(_CM A)"=@n1.Customization Gmap
   Forward Axis=@n19.RGB
   Object Bounds Max Adjustment=@n3.RGB
   Rotation Z Axis
   Normal Map=@n1.Customization Gmap
   Cropped Emissive Mask=@n1.Customization Gmap
   "Customization Crevasse (_CM G)"=@n1.Customization Gmap
   "Customization Mask UV Islands (_CM B Bottom Right)"=@n1.Customization Gmap
   -------------------
   "Customization Mask 1 (_CM B Top Left)"=@n1.Customization Gmap
   "Customization Mask 2 (_CM B Top Right)"=@n1.Customization Gmap
   "Customization Mask 3 (_CM B Bottom Left)"=@n1.Customization Gmap
   "Customization Mask (_CM RGBA)"=@n1.Customization Gmap
   --------------------
   Pre-Skinned Local Position=@n1.Customization Gmap
   Pre-Skinned Bounds=@n1.Customization Gmap
   Pre-Skinned Local Normal=@n1.Customization Gmap
   Per_ObjectMaskOverride
   isVehicle=@n20
   Side Project Vehicle=@n20
   isMech=@n20
   isCharacter=@n20
   isCharAccent
   isPickaxe=@n13
   isBackpack=@n20
   Object-SafeZone-Back=AppendVector(A=@n7.RGB, B=@n7.RGB)
   Object-SafeZone-Middle=AppendVector(A=@n8.RGB, B=@n8.RGB)
   Object-SafeZone-Front=AppendVector(A=@n9.RGB, B=@n9.RGB)
   Object-SafeZone-VP_Charcter_Preview_Switch
   Object-SafeZone-Virtual-Pivot=AppendVector(A=@n11.RGB, B=@n11.RGB)
   Wrap Holdout
  ) {materialFunction=/Game/Packages/Fortress_SharedMaterials/Base_Material_Functions/BaseCustomizationFunctions/MF_Base_Customization}
  n22 = MaterialFunctionCall!(
   Base MA=MaterialFunctionCall!(
    In=@n21.Result
    Pre-Skinned Local Position=@n1.Pre-Skinned Local Position Pixel Shader
    Pre-Skinned Local Normal=@n1.Pre-Skinned Local Normal Pixel Shader
   ) {materialFunction=/Game/Weapons/WeaponSkins/MaterialFunction/ElementalWeaponsFunction}.Result
   Pre-Skinned Local Position=@n1.Pre-Skinned Local Position Pixel Shader
  ) {materialFunction=/Game/Weapons/WeaponSkins/MaterialFunction/OverheatGunFunction}
  n23 = GetMaterialAttributes(None=MaterialFunctionCall!(Material Attributes=ShadingPathSwitch(Default=@n22.Result, Mobile=@n21.Result) {desc=Migrated from FeatureLevelSwitch}) {materialFunction=/Game/Packages/Fortnite_UniqueMaterials_1/Weapons/ADSScopeMaster}.Result) {attributeGetTypes [#2]}
  SetMaterialAttributes!(
   MaterialAttributes=@n23.MaterialAttributes
   World Position Offset=StaticSwitchParameter(
    True=MaterialFunctionCall!(WPO) {materialFunction=/Game/Packages/Fortress_SharedMaterials/Base_Material_Functions/Fortnite_HideElementsWithVertexColor}.Result
    False=@n23.World Position Offset
   ) {group=Opacity Mask, parameterName=Use Vertex Colors for Mask}
  ) {attributeSetTypes [#2]}
 ]
}
